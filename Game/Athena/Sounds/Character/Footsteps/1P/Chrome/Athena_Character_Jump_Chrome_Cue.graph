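# First-person foley Sound Cue for a player character's jump on chrome in Athena.
# A Modulator varies volume and raises pitch over a quality-level switch, whose first two slots share an eight-wave random pool and last two a four-wave pool.
# The four-wave pool is excluded from branch culling.
# It borrows the player landing attenuation, and a concurrency override allows one instance per owner, blocking new plays.
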SoundCue /Game/Athena/Sounds/Character/Footsteps/1P/Chrome/Athena_Character_Jump_Chrome_Cue
root=Modulator_2
nodes [
 {id=Modulator_2, children [QualityLevel_0], properties {pitchMax=1.3, pitchMin=1.1, volumeMax=1, volumeMin=0.7}}
 {id=QualityLevel_0, children [Random_2,Random_2,Random_3,Random_3]}
 {
  id=Random_2
  children [WavePlayer_2,WavePlayer_3,WavePlayer_4,WavePlayer_5,WavePlayer_6,WavePlayer_7,WavePlayer_8,WavePlayer_9]
  properties {weights [1,1,1,1,1,1,1,1]}
 }
 {
  id=Random_3
  children [WavePlayer_10, WavePlayer_11, WavePlayer_12, WavePlayer_13]
  properties {bShouldExcludeFromBranchCulling=true, weights [1,1,1,1]}
 }
 {id=WavePlayer_10, properties {soundWaveAssetPtr=&/Foley_FS_1P_Chrome_Jump_01}}
 {id=WavePlayer_11, properties {soundWaveAssetPtr=&/Foley_FS_1P_Chrome_Jump_02}}
 {id=WavePlayer_12, properties {soundWaveAssetPtr=&/Foley_FS_1P_Chrome_Jump_03}}
 {id=WavePlayer_13, properties {soundWaveAssetPtr=&/Foley_FS_1P_Chrome_Jump_04}}
 {id=WavePlayer_2, properties {soundWaveAssetPtr=&Foley_FS_1P_Chrome_Jump_01}}
 {id=WavePlayer_3, properties {soundWaveAssetPtr=&Foley_FS_1P_Chrome_Jump_02}}
 {id=WavePlayer_4, properties {soundWaveAssetPtr=&Foley_FS_1P_Chrome_Jump_03}}
 {id=WavePlayer_5, properties {soundWaveAssetPtr=&Foley_FS_1P_Chrome_Jump_04}}
 {id=WavePlayer_6, properties {soundWaveAssetPtr=&/Foley_FS_1P_Chrome_Jump_05}}
 {id=WavePlayer_7, properties {soundWaveAssetPtr=&/Foley_FS_1P_Chrome_Jump_06}}
 {id=WavePlayer_8, properties {soundWaveAssetPtr=&/Foley_FS_1P_Chrome_Jump_07}}
 {id=WavePlayer_9, properties {soundWaveAssetPtr=&/Foley_FS_1P_Chrome_Jump_08}}
]
properties {
 attenuationSettings=/Game/Sounds/Attenuation/Fort_Player_Land_Athena_Att
 bOverrideConcurrency=true
 maxDistance=2097152
 soundClassObject=&Fort_Player_Foley_1P_Athena
 concurrencyOverrides {bLimitToOwner=true, maxCount=1, resolutionRule=PreventNew}
}
